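; PCG instant graph that moves source actors to align with target actors, chosen by editor selection or actor references, with per-axis referential options.
; Multiple targets always combine into one point before Align Points runs; empty sources or targets print an error.
; With AlignSourcesAsAGroup, sources are combined, aligned as one, and the result is re-applied through inverted-transform math to keep relative layout.
; Teleport Actors And Components applies the final transforms.
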(graph Instant_Place_AlignTo
  (params (Source_UseSelectedActors :type Bool :default true) (Source_ActorReferences :type (Array (SoftObject /Script/Engine.Actor))) (Target_UseSelectedActors :type Bool) (Target_ActorReferences :type (Array (SoftObject /Script/Engine.Actor))) (AlignSourcesAsAGroup :type Bool) (SpatialReferential :type (Enum /Script/PCG.EPCGAlignPointsSpatialReferential) :default LocalToTarget) (Align_XAxis :type Bool :default true) (Align_YAxis :type Bool :default true) (Align_ZAxis :type Bool :default true) (XAxis_Source_Referential :type (Enum /Script/PCG.EPCGAlignPointsAxisReferential) :default Center) (XAxis_Target_Referential :type (Enum /Script/PCG.EPCGAlignPointsAxisReferential) :default Center) (YAxis_Source_Referential :type (Enum /Script/PCG.EPCGAlignPointsAxisReferential) :default Center) (YAxis_Target_Referential :type (Enum /Script/PCG.EPCGAlignPointsAxisReferential) :default Center) (ZAxis_Source_Referential :type (Enum /Script/PCG.EPCGAlignPointsAxisReferential) :default Center) (ZAxis_Target_Referential :type (Enum /Script/PCG.EPCGAlignPointsAxisReferential) :default Center))
  (bind DataNum_0 (Data%20Count))
  (bind AttributeMathsOp_0 (Attribute%20Maths%20Op :Operation Subtract))
  (bind Branch_3 (Branch))
  (bind Branch_0 (Branch)) ; Target to align to
  (bind DataNum_1 (Data%20Count))
  (bind Branch_6 (Branch))
  (bind Branch_4 (Branch :bOutputToB-param AlignSourcesAsAGroup)) ; Source(s) to align
  (bind CombinePoints_2 (Combine%20Points :bUseFirstPointTransform false)) ; Always combine all targets into one if multiple
  (bind Select_0 (Select :bUseInputB-param AlignSourcesAsAGroup))
  ;[0 0 0 0.5] Source(s) actors to align
  (bind Subgraph_1 (Subgraph :Graph /PCGPrimitives/Instants/Shared/Input_ActorRefsOrSelection.Input_ActorRefsOrSelection :ActorReferences Source_ActorReferences :UseSelectedActors Source_UseSelectedActors))
  (bind GetActorData_1 (Get%20Actor%20Data :ActorSelector (struct :ActorFilter FromInput :ActorReferenceSelector ActorReferences) :bMergeSinglePointData true :Mode GetSinglePoint))
  (bind Gather_1 (Gather))
  ;[0 0 0 0.5] Target(s) actors to align source(s) to
  (bind Subgraph_2 (Subgraph :Graph /PCGPrimitives/Instants/Shared/Input_ActorRefsOrSelection.Input_ActorRefsOrSelection :ActorReferences Target_ActorReferences :UseSelectedActors Target_UseSelectedActors))
  (bind GetActorData_2 (Get%20Actor%20Data :ActorSelector (struct :ActorFilter FromInput :ActorReferenceSelector ActorReferences) :bMergeSinglePointData true :Mode GetSinglePoint))
  (bind Gather_2 (Gather))
  ;[0 0 0 0.5] Combine sources into a single point to align and re-apply its aligned transform back onto sources.
  (bind CombinePoints_0 (Combine%20Points :bUseFirstPointTransform false)) ; Combine sources only if desired\nby enabling AlignSourcesAsAGroup parameter
  (bind AttributeTransformOp_1 (Attribute%20Transform%20Op :InputSource1 $Transform :Operation Invert))
  (bind AttributeTransformOp_2 (Attribute%20Transform%20Op :InputSource1 $Transform :InputSource2 $Transform)) ; Aply Inverted combined transform to source points
  (bind AttributeTransformOp_0 (Attribute%20Transform%20Op :InputSource1 $Transform :InputSource2 $Transform)) ; Apply aligned transform of combined point over source points.
  ;[0 0 0 0.5] Align source(s) to target
  (bind AlignPoints_0 (Align%20Points :SpatialReferential Target :XAxis (struct :bApplyToAxis true :SourceReferential Min :TargetReferential Max) :YAxis (struct :bApplyToAxis true) :ZAxis (struct :bApplyToAxis true) :SpatialReferential-param SpatialReferential :XAxis/bApplyToAxis Align_XAxis :XAxis/SourceReferential XAxis_Source_Referential :XAxis/TargetReferential XAxis_Target_Referential :YAxis/bApplyToAxis Align_YAxis :YAxis/SourceReferential YAxis_Source_Referential :YAxis/TargetReferential YAxis_Target_Referential :ZAxis/bApplyToAxis Align_ZAxis :ZAxis/SourceReferential ZAxis_Source_Referential :ZAxis/TargetReferential ZAxis_Target_Referential))
  ;[0 0 0 0.5] Moving source(s) to their aligned position
  (bind TeleportActorsAndComponents_0 (Teleport%20Actors%20And%20Components :ObjectReferenceAttribute ActorReference))
  ;[1 0 0 0.5] No target error
  (bind Print%20String_1 (Print%20String :bPrefixWithNode false :bPrintPerComponent false :PrintString "No target actor selected or referenced to align to." :Verbosity Error))
  ;[1 0 0 0.5] No source error
  (bind Print%20String_2 (Print%20String :bPrefixWithNode false :bPrintPerComponent false :PrintString "No source actor selected or referenced to align to." :Verbosity Error))
  (connect Branch_4 Output%20A AlignPoints_0 Source)
  (connect CombinePoints_0 Out AlignPoints_0 Source)
  (connect Branch_0 Output%20A AlignPoints_0 Target)
  (connect CombinePoints_2 Out AlignPoints_0 Target)
  (connect DataNum_0 Out AttributeMathsOp_0 InA)
  (connect AttributeTransformOp_2 Out AttributeTransformOp_0 InA)
  (connect AlignPoints_0 Out AttributeTransformOp_0 InB)
  (connect CombinePoints_0 Out AttributeTransformOp_1 In)
  (connect Branch_4 Output%20B AttributeTransformOp_2 InA)
  (connect AttributeTransformOp_1 Out AttributeTransformOp_2 InB)
  (connect AttributeMathsOp_0 Out Branch_0 bOutputToB)
  (connect Branch_3 Output%20B Branch_0 In)
  (connect DataNum_0 Out Branch_3 bOutputToB)
  (connect Gather_2 Out Branch_3 In)
  (connect Branch_6 Output%20B Branch_4 In)
  (connect DataNum_1 Out Branch_6 bOutputToB)
  (connect Gather_1 Out Branch_6 In)
  (connect Branch_4 Output%20B CombinePoints_0 In)
  (connect Branch_0 Output%20B CombinePoints_2 In)
  (connect Gather_2 Out DataNum_0 In)
  (connect Gather_1 Out DataNum_1 In)
  (connect GetActorData_1 Out Gather_1 In)
  (connect Subgraph_1 SelectedActors Gather_1 In)
  (connect GetActorData_2 Out Gather_2 In)
  (connect Subgraph_2 SelectedActors Gather_2 In)
  (connect Subgraph_1 ActorReferences GetActorData_1 In)
  (connect Subgraph_2 ActorReferences GetActorData_2 In)
  (connect Branch_3 Output%20A Print%20String_1 In)
  (connect Branch_6 Output%20A Print%20String_2 In)
  (connect AlignPoints_0 Out Select_0 Input%20A)
  (connect AttributeTransformOp_0 Out Select_0 Input%20B)
  (connect Select_0 Out TeleportActorsAndComponents_0 In))
