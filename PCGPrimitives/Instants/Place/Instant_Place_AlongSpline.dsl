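; PCG graph that places weighted object references along splines of selected or referenced actors.
; It samples by count, reference count or distance, assigns objects by index or weight, then applies global and random transforms.
; Points branch by class: StaticMeshActors for meshes, actors for Blueprints, and for PCG Data Assets an empty actor per point receiving the assembly's meshes via a loop subgraph.
; Missing splines or valid references print errors.
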(graph Instant_Place_AlongSpline
  (params (Count :type Int :default 10) (UseDistanceSamplingMode :type Bool) (Distance :type Double :default 1e+02) (FitToCurve :type Bool :default true) (StartOffset :type Double) (EndOffset :type Double) (MaxRandomOffset :type Double) (SamplingSeed :type Int :default 2145617851) (GlobalPositionOffset :type Vector) (GlobalRotation :type Rotator) (AbsoluteGlobalRotation :type Bool) (GlobalScaleFactor :type Vector :default (1.0 1.0 1.0)) (RandomPositionOffsetMin :type Vector) (RandomPositionOffsetMax :type Vector) (RandomRotationMin :type Rotator) (RandomRotationMax :type Rotator) (RandomScaleFactorMin :type Vector :default (1.0 1.0 1.0)) (RandomScaleFactorMax :type Vector :default (1.0 1.0 1.0)) (UniformRandomScaleFactor :type Bool :default true) (RandomSeed :type Int :default 2118139241) (ObjectReferences :type (Array (Struct /PCG/EdMode/Resources/PCGEdMode_WeightedObjectReference.PCGEdMode_WeightedObjectReference))) (UseObjectRefsCount :type Bool :default true) (UseSelectedActors :type Bool :default true) (ActorReferences :type (Array (SoftObject /Script/Engine.Actor))))
  (bind CreateConstant_0 (Create%20Constant :AttributeTypes (struct :Type Integer32 :Int32Value 2)))
  (bind AttributeMathsOp_0 (Attribute%20Maths%20Op :Operation Subtract :InB UseDistanceSamplingMode))
  (bind AttributeMathsOp_1 (Attribute%20Maths%20Op :Operation Divide :InA MaxRandomOffset))
  (bind Subgraph_0 (Subgraph :Graph /PCGPrimitives/Instants/Shared/Input_ActorRefsOrSelection.Input_ActorRefsOrSelection :ActorReferences ActorReferences :UseSelectedActors UseSelectedActors))
  (bind GetActorData_1 (Get%20Actor%20Data :ActorSelector (struct :ActorFilter FromInput :ActorReferenceSelector ActorReferences) :bMergeSinglePointData true :Mode GetSinglePoint))
  (bind Gather_1 (Gather))
  (bind GetSplineData_0 (Get%20Spline%20Data :ActorSelector (struct :ActorFilter FromInput :ActorReferenceSelector ActorReference) :bIgnorePCGGeneratedComponents false))
  (bind DataNum_0 (Data%20Count))
  (bind Branch_5 (Branch))
  (bind RemoveEmptyData_0 (Remove%20Empty%20Data :In ObjectReferences))
  (bind AttributeFilter_0 (Filter%20Attribute%20Elements :AttributeTypes (struct :Type SoftObjectPath) :bGenerateOutputDataEvenIfEmpty false :bUseConstantThreshold true :Operator NotEqual :TargetAttribute Object :ThresholdAttribute @Last))
  (bind GetElementsCount_0 (Get%20Element%20Count))
  (bind AttributeMathsOp_3 (Attribute%20Maths%20Op))
  (bind AttributeMathsOp_4 (Attribute%20Maths%20Op :InA Count))
  (bind Select_0 (Select :bUseInputB-param UseObjectRefsCount))
  (bind SplineSampler_0 (Spline%20Sampler :SamplerParams (struct :Mode NumberOfSamples :DistanceIncrement 5e+02 :bUnbounded true :PointSteepness 1.0) :bFitToCurve FitToCurve :DistanceIncrement Distance :EndOffset EndOffset :Seed-param SamplingSeed :StartOffset StartOffset))
  (bind AttributeMathsOp_5 (Attribute%20Maths%20Op :Operation Subtract))
  (bind AttributeStringOp_0 (Attribute%20String%20Op))
  (bind FilterElementsByIndex_0 (Filter%20Elements%20By%20Index :bInvertFilter true :bSelectIndicesByInput false))
  (bind BoundsModifier_1 (Bounds%20Modifier :BoundsMax (5e+01 5e+01 5e+01) :BoundsMin (-5e+01 -5e+01 -5e+01) :Mode Set))
  (bind Branch_7 (Branch :bOutputToB-param UseObjectRefsCount))
  (bind MatchAndSetAttributes_1 (Match%20And%20Set%20Attributes :bUseWeightAttribute true :WeightAttribute Weight :Match%20Data ObjectReferences))
  (bind CopyAttributes_0 (Copy%20Attributes :InputSource $Index :OutputTarget Index :Source ObjectReferences :Target ObjectReferences))
  (bind MatchAndSetAttributes_0 (Match%20And%20Set%20Attributes :bKeepUnmatched false :bMatchAttributes true :bUseWeightAttribute true :InputAttribute $Index :MatchAttribute Index :WeightAttribute Weight))
  (bind TransformPoints_1 (Transform%20Points :bUniformScale false :bAbsoluteRotation-param AbsoluteGlobalRotation :OffsetMax-param GlobalPositionOffset :OffsetMin-param GlobalPositionOffset :RotationMax-param GlobalRotation :RotationMin-param GlobalRotation :ScaleMax-param GlobalScaleFactor :ScaleMin-param GlobalScaleFactor))
  (bind TransformPoints_2 (Transform%20Points :bUniformScale-param UniformRandomScaleFactor :OffsetMax-param RandomPositionOffsetMax :OffsetMin-param RandomPositionOffsetMin :RotationMax-param RandomRotationMax :RotationMin-param RandomRotationMin :ScaleMax-param RandomScaleFactorMax :ScaleMin-param RandomScaleFactorMin :Seed-param RandomSeed))
  (bind GetClassFromAttribute_1 (Get%20Class%20From%20Attribute :InputSource Object :OutputTarget ObjectType))
  (bind CreateConstant_5 (Create%20Constant :AttributeTypes (struct :Type String :StringValue "PCGdataAsset")))
  (bind AttributeFilter_4 (Filter%20Attribute%20Elements :bGenerateOutputDataEvenIfEmpty false :bWarnOnDataMissingAttribute false :Operator Matches :TargetAttribute ObjectType :ThresholdAttribute @Last))
  (bind DataNum_2 (Data%20Count))
  (bind Branch_10 (Branch))
  (bind AttributePartition_0 (Attribute%20Partition :PartitionAttributeSelectors [Object]))
  (bind ExtractAttribute_0 (Extract%20Attribute%20at%20Index :bForceOutputAttributeToBeInElementsDomain true :InputSource Object))
  (bind AttributeFilter_2 (Filter%20Attribute%20Elements :AttributeTypes (struct :Type String) :bUseConstantThreshold true :Operator Matches :TargetAttribute Object :ThresholdAttribute @Last)) ; FAIL SAFE if asset is empty
  (bind PCGLoadDataAsset_0 (Load%20PCG%20Data%20Asset :bLoadFromInput true))
  (bind AttributePartition_1 (Attribute%20Partition :PartitionAttributeSelectors [$Index]))
  (bind ToPoint_0 (To%20Point))
  (bind FilterDataByTag_0 (Filter%20Data%20By%20Tag :SelectedTags "points"))
  (bind CopyPoints_0 (Copy%20Points :AttributeInheritance SourceOnly :bCopyEachSourceOnEveryTarget false :TagInheritance Source))
  (bind CreateConstant_3 (Create%20Constant :AttributeTypes (struct :Type String :StringValue "Blueprint")))
  (bind AttributeFilter_5 (Filter%20Attribute%20Elements :bGenerateOutputDataEvenIfEmpty false :bWarnOnDataMissingAttribute false :Operator Matches :TargetAttribute ObjectType :ThresholdAttribute @Last))
  (bind DataNum_3 (Data%20Count))
  (bind Branch_11 (Branch))
  (bind SpawnActor_0 (Spawn%20Actor :bSpawnByAttribute true :SpawnAttribute Object))
  (bind CreateConstant_4 (Create%20Constant :AttributeTypes (struct :Type String :StringValue "StaticMesh")))
  (bind AttributeFilter_3 (Filter%20Attribute%20Elements :bGenerateOutputDataEvenIfEmpty false :bWarnOnDataMissingAttribute false :Operator Matches :TargetAttribute ObjectType :ThresholdAttribute @Last))
  (bind DataNum_1 (Data%20Count))
  (bind Branch_9 (Branch))
  (bind SpawnActor_3 (Spawn%20Actor :AttachOptions NotAttached :bAllowTemplateActorEditing true :SpawnedActorPropertyOverrideDescriptions [(struct :InputSource Object :PropertyTarget "StaticMeshComponent.StaticMesh")] :TemplateActor (class:/Script/Engine.StaticMeshActor) :TemplateActorClass /Script/Engine.StaticMeshActor))
  (bind SpawnActor_4 (Spawn%20Actor :AttachOptions NotAttached :bAllowTemplateActorEditing true :TemplateActor (class:/Script/Engine.Actor) :TemplateActorClass /Script/Engine.Actor))
  (bind Loop_0 (Loop :Graph SpawnAssembliesWithTargetActorLoop))
  (bind AttributeMathsOp_6 (Attribute%20Maths%20Op))
  (bind AttributeMathsOp_7 (Attribute%20Maths%20Op))
  (bind AttributeCompareOp_1 (Attribute%20Compare%20Op))
  (bind Branch_13 (Branch))
  ;[1 0 0 0.5] Error: No Spline Actor Selected
  (bind Print%20String_0 (Print%20String :bPrefixWithNode false :bPrintPerComponent false :PrintString "No Spline Actor Selected" :Verbosity Error))
  ;[1 0 0 0.5] Error: No Valid Object Ref to Spawn
  (bind Print%20String_1 (Print%20String :bPrefixWithNode false :bPrintPerComponent false :PrintString "No valid obejct reference to spawn. Use StaticMesh, Blueprint class and/or PCG Data Asset references." :Verbosity Error))
  (connect AttributeMathsOp_7 Out AttributeCompareOp_1 InA)
  (connect RemoveEmptyData_0 Out AttributeFilter_0 In)
  (connect ExtractAttribute_0 Out AttributeFilter_2 In)
  (connect CreateConstant_4 Out AttributeFilter_3 Filter)
  (connect GetClassFromAttribute_1 Out AttributeFilter_3 In)
  (connect CreateConstant_5 Out AttributeFilter_4 Filter)
  (connect GetClassFromAttribute_1 Out AttributeFilter_4 In)
  (connect CreateConstant_3 Out AttributeFilter_5 Filter)
  (connect GetClassFromAttribute_1 Out AttributeFilter_5 In)
  (connect CreateConstant_0 Out AttributeMathsOp_0 InA)
  (connect GetElementsCount_0 Out AttributeMathsOp_3 InA)
  (connect Select_0 Out AttributeMathsOp_5 InA)
  (connect DataNum_1 Out AttributeMathsOp_6 InA)
  (connect DataNum_2 Out AttributeMathsOp_6 InB)
  (connect AttributeMathsOp_6 Out AttributeMathsOp_7 InA)
  (connect DataNum_3 Out AttributeMathsOp_7 InB)
  (connect Branch_10 Output%20B AttributePartition_0 In)
  (connect AttributePartition_0 Out AttributePartition_1 In)
  (connect AttributeMathsOp_5 Out AttributeStringOp_0 InB)
  (connect FilterElementsByIndex_0 In%20Filter BoundsModifier_1 In)
  (connect DataNum_0 Out Branch_5 bOutputToB)
  (connect GetSplineData_0 Out Branch_5 In)
  (connect BoundsModifier_1 Out Branch_7 In)
  (connect DataNum_1 Out Branch_9 bOutputToB)
  (connect AttributeFilter_3 InsideFilter Branch_9 In)
  (connect DataNum_2 Out Branch_10 bOutputToB)
  (connect AttributeFilter_4 InsideFilter Branch_10 In)
  (connect DataNum_3 Out Branch_11 bOutputToB)
  (connect AttributeFilter_5 InsideFilter Branch_11 In)
  (connect AttributeCompareOp_1 Out Branch_13 bOutputToB)
  (connect AttributeMathsOp_6 Out Branch_13 In)
  (connect FilterDataByTag_0 InsideFilter CopyPoints_0 Source)
  (connect AttributePartition_1 Out CopyPoints_0 Target)
  (connect GetSplineData_0 Out DataNum_0 In)
  (connect AttributeFilter_3 InsideFilter DataNum_1 In)
  (connect AttributeFilter_4 InsideFilter DataNum_2 In)
  (connect AttributeFilter_5 InsideFilter DataNum_3 In)
  (connect AttributePartition_0 Out ExtractAttribute_0 In)
  (connect ToPoint_0 Out FilterDataByTag_0 In)
  (connect SplineSampler_0 Out FilterElementsByIndex_0 In)
  (connect AttributeStringOp_0 Out FilterElementsByIndex_0 SelectedIndices)
  (connect GetActorData_1 Out Gather_1 In)
  (connect Subgraph_0 SelectedActors Gather_1 In)
  (connect Subgraph_0 ActorReferences GetActorData_1 In)
  (connect TransformPoints_2 Out GetClassFromAttribute_1 In)
  (connect AttributeFilter_0 InsideFilter GetElementsCount_0 In)
  (connect Gather_1 Out GetSplineData_0 In)
  (connect CopyPoints_0 Out Loop_0 AssemblyData)
  (connect SpawnActor_4 Out Loop_0 TargetActor)
  (connect Branch_7 Output%20B MatchAndSetAttributes_0 In)
  (connect CopyAttributes_0 Out MatchAndSetAttributes_0 Match%20Data)
  (connect Branch_7 Output%20A MatchAndSetAttributes_1 In)
  (connect AttributeFilter_2 OutsideFilter PCGLoadDataAsset_0 In)
  (connect Branch_5 Output%20A Print%20String_0 In)
  (connect Branch_13 Output%20B Print%20String_1 In)
  (connect AttributeMathsOp_4 Out Select_0 Input%20A)
  (connect AttributeMathsOp_3 Out Select_0 Input%20B)
  (connect Branch_11 Output%20B SpawnActor_0 In)
  (connect Branch_9 Output%20B SpawnActor_3 In)
  (connect AttributePartition_1 Out SpawnActor_4 In)
  (connect AttributeMathsOp_1 Out SplineSampler_0 MaxRandomOffsetNormalized)
  (connect AttributeMathsOp_0 Out SplineSampler_0 Mode)
  (connect Select_0 Out SplineSampler_0 NumSamples)
  (connect Branch_5 Output%20B SplineSampler_0 Spline)
  (connect PCGLoadDataAsset_0 Out ToPoint_0 In)
  (connect MatchAndSetAttributes_0 Out TransformPoints_1 In)
  (connect MatchAndSetAttributes_1 Out TransformPoints_1 In)
  (connect TransformPoints_1 Out TransformPoints_2 In))

(subgraph SpawnAssembliesWithTargetActorLoop
  (inputs (AssemblyData :type Point) (TargetActor :type Point))
  (bind ExtractAttribute_0 (Extract%20Attribute%20at%20Index :bForceOutputAttributeToBeInElementsDomain true :InputSource ActorReference))
  (bind StaticMeshSpawner_0 (Static%20Mesh%20Spawner :InstanceDataPackerParameters (class:/Script/PCG.PCGInstanceDataPackerByAttribute :AttributeSelectors [$Density $Extents $Color]) :InstanceDataPackerType /Script/PCG.PCGInstanceDataPackerByAttribute :MeshSelectorParameters (class:/Script/PCG.PCGMeshSelectorByAttribute :AttributeName Mesh :TemplateDescriptor (struct :bHasPerInstanceHitProxies true)) :MeshSelectorType /Script/PCG.PCGMeshSelectorByAttribute))
  (connect DefaultInputNode TargetActor ExtractAttribute_0 In)
  (connect DefaultInputNode AssemblyData StaticMeshSpawner_0 In)
  (connect ExtractAttribute_0 Out StaticMeshSpawner_0 TargetActor))
