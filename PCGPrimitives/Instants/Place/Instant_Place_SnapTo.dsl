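; PCG instant graph that snaps source actors (selected or referenced) onto raycast hits, with target, tag, class and landscape filters.
; Rays run from each source toward the first target or along a custom direction, or from the editor viewport camera.
; Camera mode can move sources as a group around their combined center or an indexed member.
; Successful hits teleport the actors, optionally position-only; misses print an error naming the actors.
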(graph Instant_Place_SnapTo
  (params (Source_UseSelectedActors :type Bool :default true) (Source_ActorReferences :type (Array (SoftObject /Script/Engine.Actor))) (Target_UseSelectedActors :type Bool) (Target_ActorReferences :type (Array (SoftObject /Script/Engine.Actor))) (Target_UseEditorViewportCamera :type Bool) (SnapPositionOnly :type Bool) (SnapAsAGroup :type Bool) (SnapAsAGroupIndex :type Int :default -1) (TargetActorTagsList :type String) (TargetActorTagsBehavior :type Int :default 1) (TargetFilterActorsBehavior :type Int :default 1) (TargetActorClassFilterBehavior :type Int :default 1) (TargetActorClassFilter :type (Struct /Script/CoreUObject.SoftClassPath)) (LandscapeHitsFilterBehavior :type Int :default 1) (RayDirection :type Vector) (UseCustomRayDirection :type Bool))
  (bind Branch_3 (Branch :bOutputToB-param Target_UseEditorViewportCamera))
  (bind Branch_8 (Branch :bOutputToB-param UseCustomRayDirection))
  (bind FilterElementsByIndex_1 (Filter%20Elements%20By%20Index :bSelectIndicesByInput false :SelectedIndices "0"))
  (bind AttributeMathsOp_0 (Attribute%20Maths%20Op :InputSource1 $Position :InputSource2 $Position :Operation Subtract :OutputTarget ToTargetVector))
  (bind CopyAttributes_6 (Copy%20Attributes :InputSource @Last :OutputTarget ToTargetVector :Source RayDirection))
  (bind AttributeVectorOp_0 (Attribute%20Vector%20Op :InputSource1 ToTargetVector :Operation Normalize))
  (bind PointToAttributeSet_0 (Point%20To%20Attribute%20Set))
  (bind DeleteAttributes_0 (Delete%20Attributes :Operation KeepSelectedAttributes :SelectedAttributes "ActorReference"))
  (bind CopyAttributes_4 (Copy%20Attributes :InputSource $Transform :OutputTarget SourceTransform))
  (bind Branch_4 (Branch :bOutputToB-param SnapAsAGroup))
  (bind AttributeMathsOp_1 (Attribute%20Maths%20Op :InputSource1 $Position :InputSource2 $Position :Operation Set))
  (bind CopyAttributes_5 (Copy%20Attributes :InputSource ImpactResult :OutputTarget ImpactResult))
  (bind Branch_1 (Branch :bOutputToB-param SnapPositionOnly))
  (bind CopyAttributes_1 (Copy%20Attributes :InputSource SourceTransform.Rotation :OutputTarget $Rotation))
  (bind CopyAttributes_2 (Copy%20Attributes :InputSource SourceTransform.Scale :OutputTarget $Scale))
  (bind AttributeCompareOp_0 (Attribute%20Compare%20Op :Operation NotEqual :InA SnapAsAGroupIndex))
  (bind Branch_5 (Branch))
  (bind CombinePoints_1 (Combine%20Points))
  (bind FilterElementsByIndex_2 (Filter%20Elements%20By%20Index :bSelectIndicesByInput false :SelectedIndices-param SnapAsAGroupIndex))
  (bind AttributeMathsOp_4 (Attribute%20Maths%20Op :InputSource1 $Position :InputSource2 $Position :Operation Subtract))
  (bind AttributeMathsOp_5 (Attribute%20Maths%20Op :InA SnapPositionOnly))
  (bind CopyPoints_0 (Copy%20Points :ColorInheritance Source :RotationInheritance Target :ScaleInheritance Source :SeedInheritance Source :TagInheritance Source))
  (bind AttributeFilter_0 (Filter%20Attribute%20Elements :AttributeTypes (struct :Type Boolean :BoolValue true) :bUseConstantThreshold true :Operator Equal :TargetAttribute ImpactResult :ThresholdAttribute @Last))
  (bind GetElementsCount_0 (Get%20Element%20Count))
  (bind Branch_2 (Branch))
  ;[0 0 0 0.5] Source(s) actors to snap
  (bind Subgraph_1 (Subgraph :Graph /PCGPrimitives/Instants/Shared/Input_ActorRefsOrSelection.Input_ActorRefsOrSelection :ActorReferences Source_ActorReferences :UseSelectedActors Source_UseSelectedActors))
  (bind GetActorData_1 (Get%20Actor%20Data :ActorSelector (struct :ActorFilter FromInput :ActorReferenceSelector ActorReferences) :bMergeSinglePointData true :Mode GetSinglePoint))
  (bind Gather_1 (Gather))
  ;[0 0 0 0.5] Target(s) actors to snap source(s) to
  (bind Subgraph_2 (Subgraph :Graph /PCGPrimitives/Instants/Shared/Input_ActorRefsOrSelection.Input_ActorRefsOrSelection :ActorReferences Target_ActorReferences :UseSelectedActors Target_UseSelectedActors))
  (bind GetActorData_2 (Get%20Actor%20Data :ActorSelector (struct :ActorFilter FromInput :ActorReferenceSelector ActorReferences) :bMergeSinglePointData true :Mode GetSinglePoint))
  (bind Gather_2 (Gather))
  ;[0 0 0 0.5] Raycast on targets with optional filtering
  (bind CopyAttributes_0 (Copy%20Attributes :InputSource $Transform :OutputTarget SourceTransform))
  (bind WorldRaycast_0 (World%20Raycast :bKeepOriginalPointOnMiss true :bOverrideRayDirections true :bUnbounded true :RayDirectionAttribute ToTargetVector :WorldQueryParams (struct :bGetImpact true :bTraceComplex true :ActorTagFilter Include :ActorClassFilter Include :ActorFilterFromInput Include :ActorFilterInputSource ActorReference) :ActorClass TargetActorClassFilter :ActorClassFilter TargetActorClassFilterBehavior :ActorFilterFromInput TargetFilterActorsBehavior :ActorTagFilter TargetActorTagsBehavior :ActorTagsList TargetActorTagsList :SelectLandscapeHits LandscapeHitsFilterBehavior))
  ;[0 0 0 0.5] Use Editor Viewport Camera as snapping ray direction vector on anything or defined target(s)
  (bind Subgraph_0 (Subgraph :Graph /PCGPrimitives/Instants/Shared/Input_EditorViewportCamera.Input_EditorViewportCamera :UseEditorViewportCamera Target_UseEditorViewportCamera))
  ;[0 0 0 0.5] Raycast on targets with optional filtering
  (bind WorldRaycast_1 (World%20Raycast :bKeepOriginalPointOnMiss true :bUnbounded true :RayDirectionAttribute ToTargetVector :WorldQueryParams (struct :bGetImpact true :bTraceComplex true :ActorTagFilter Include :ActorClassFilter Include :ActorFilterFromInput Include :ActorFilterInputSource ActorReference) :ActorClass TargetActorClassFilter :ActorClassFilter TargetActorClassFilterBehavior :ActorFilterFromInput TargetFilterActorsBehavior :ActorTagFilter TargetActorTagsBehavior :ActorTagsList TargetActorTagsList :SelectLandscapeHits LandscapeHitsFilterBehavior))
  ;[0 0 0 0.5] Moving source(s) to valid destination
  (bind TeleportActorsAndComponents_0 (Teleport%20Actors%20And%20Components :ObjectReferenceAttribute ActorReference))
  ;[1 0 0 0.5] Error: Could not snap the following actors
  (bind AttributeReduce_0 (Attribute%20Reduce :InputSource ActorReference :Operation Join))
  (bind AttributeStringOp_0 (Attribute%20String%20Op))
  (bind Print%20String_0 (Print%20String :bPrefixWithNode false :bPrintPerComponent false :PrintString "No Actor Selected" :Verbosity Error))
  (connect Branch_1 Output%20A AttributeFilter_0 In)
  (connect CopyAttributes_2 Out AttributeFilter_0 In)
  (connect CopyPoints_0 Out AttributeFilter_0 In)
  (connect FilterElementsByIndex_1 In%20Filter AttributeMathsOp_0 InA)
  (connect Branch_8 Output%20A AttributeMathsOp_0 InB)
  (connect Branch_4 Output%20A AttributeMathsOp_1 InA)
  (connect WorldRaycast_1 Out AttributeMathsOp_1 InB)
  (connect CopyAttributes_4 Out AttributeMathsOp_4 InA)
  (connect CombinePoints_1 Out AttributeMathsOp_4 InB)
  (connect FilterElementsByIndex_2 In%20Filter AttributeMathsOp_4 InB)
  (connect Branch_2 Output%20B AttributeReduce_0 In)
  (connect AttributeReduce_0 Out AttributeStringOp_0 InB)
  (connect AttributeMathsOp_0 Out AttributeVectorOp_0 In)
  (connect CopyAttributes_6 Out AttributeVectorOp_0 In)
  (connect CopyAttributes_5 Out Branch_1 In)
  (connect WorldRaycast_0 Out Branch_1 In)
  (connect GetElementsCount_0 Out Branch_2 bOutputToB)
  (connect AttributeFilter_0 OutsideFilter Branch_2 In)
  (connect Gather_1 Out Branch_3 In)
  (connect CopyAttributes_4 Out Branch_4 In)
  (connect AttributeCompareOp_0 Out Branch_5 bOutputToB)
  (connect Branch_4 Output%20B Branch_5 In)
  (connect Branch_3 Output%20A Branch_8 In)
  (connect Branch_5 Output%20A CombinePoints_1 In)
  (connect AttributeVectorOp_0 Out CopyAttributes_0 Source)
  (connect AttributeVectorOp_0 Out CopyAttributes_0 Target)
  (connect Branch_1 Output%20B CopyAttributes_1 Source)
  (connect Branch_1 Output%20B CopyAttributes_1 Target)
  (connect CopyAttributes_1 Out CopyAttributes_2 Source)
  (connect CopyAttributes_1 Out CopyAttributes_2 Target)
  (connect Branch_3 Output%20B CopyAttributes_4 Target)
  (connect WorldRaycast_1 Out CopyAttributes_5 Source)
  (connect AttributeMathsOp_1 Out CopyAttributes_5 Target)
  (connect Branch_8 Output%20B CopyAttributes_6 Target)
  (connect AttributeMathsOp_5 Out CopyPoints_0 RotationInheritance)
  (connect AttributeMathsOp_4 Out CopyPoints_0 Source)
  (connect WorldRaycast_1 Out CopyPoints_0 Target)
  (connect PointToAttributeSet_0 Out DeleteAttributes_0 In)
  (connect Gather_2 Out FilterElementsByIndex_1 In)
  (connect Branch_5 Output%20B FilterElementsByIndex_2 In)
  (connect GetActorData_1 Out Gather_1 In)
  (connect Subgraph_1 SelectedActors Gather_1 In)
  (connect GetActorData_2 Out Gather_2 In)
  (connect Subgraph_2 SelectedActors Gather_2 In)
  (connect Subgraph_1 ActorReferences GetActorData_1 In)
  (connect Subgraph_2 ActorReferences GetActorData_2 In)
  (connect AttributeFilter_0 OutsideFilter GetElementsCount_0 In)
  (connect Gather_2 Out PointToAttributeSet_0 In)
  (connect Branch_2 Output%20B Print%20String_0 In)
  (connect AttributeStringOp_0 Out Print%20String_0 PrintString)
  (connect AttributeFilter_0 InsideFilter TeleportActorsAndComponents_0 In)
  (connect DeleteAttributes_0 Out WorldRaycast_0 FilterActors)
  (connect CopyAttributes_0 Out WorldRaycast_0 Origins)
  (connect DeleteAttributes_0 Out WorldRaycast_1 FilterActors)
  (connect Subgraph_0 EditorViewportCameras WorldRaycast_1 Origins)
  (connect Subgraph_0 ActiveCameraForwardVector WorldRaycast_1 RayDirection))
